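# NS_Photon is the Niagara effect for the Photon vehicle boost cosmetic, emitted from each filtered socket on the car's skeletal mesh.
# Two ribbon emitters trail backward with inherited source movement, one adding out-of-phase sine "noodle" positions.
# A short-lived mesh/sprite emitter sits at the ports, while Rings001 fires spawn-group-driven rings and jittered particles rearward.
# VC_PaintedVariantsOverride sets per-paint colors, with extra R1/R2 layers on the rings and second ribbon.
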NiagaraSystem /VehicleCosmetics/Boosts/Boost_Photon/FX/NS_Photon
dependencies {
 usedDataInterfaces [&NiagaraDataInterfaceCurve, &NiagaraDataInterfaceArrayFloat3, &NiagaraDataInterfaceSkeletalMesh]
 usedDynamicInputs []
 usedModules [
  &self:SystemSpawnScript.SystemScriptSource.SystemScriptGraph.NiagaraNodeAssignment_0.SetVariables_#1
  &SystemState
 ]
}
emitters {
 L_Second_Ribbon001 {
  emitterUpdateScript [
   &EmitterState()
   &VC_FilteredSocketCount()
   &VC_RateQualityHandler(Low Spawn Rate=45, Medium Spawn Rate=45, Use Custom Rate?=true)
   &VC_SpawnRibbon_v1(Number of Ribbons=$Emitter.PortCount, SpawnRate=30)
  ]
  particleSpawnScript [
   &VC_PaintedVariantsOverride(
    Black=<rgba 0.02 0.02 0.02 1>
    BurntSienna=<rgba 1.5 0.5039 0.1719 1>
    Cobalt=<rgba 0.013 0.2122 1 1>
    Crimson=<rgba 1 0.1529 0.1356 1>
    DefaultColor=<rgba 5 3.8337 0.5 1>
    ForestGreen=<rgba 0.3 1.5 0.4274 1>
    Grey=<rgba 0.42 0.42 0.42 1>
    Lime=<rgba 0.6842 1.15 0.115 1>
    Orange=<rgba 1.8 0.603 0.09 1>
    Pink=<rgba 1.5 0.1875 1.5 1>
    Purple=<rgba 0.3825 0.0375 0.9 1>
    R1 Black:LinearColor=<rgba 0.009 0.009 0.009 1>
    R1 BurntSienna:LinearColor=<rgba 0.5 0.17 0.06 1>
    R1 Cobalt:LinearColor=<rgba 0 0.04 0.8 1>
    R1 Crimson:LinearColor=<rgba 1.2 0.036 0.0554 1>
    R1 DefaultColor=<rgba 1 0.7 0.6 1>
    R1 ForestGreen:LinearColor=<rgba 0.0817 1 0.05 1>
    R1 Grey:LinearColor=<rgba 0.15 0.15 0.15 1>
    R1 Lime:LinearColor=<rgba 0.4465 0.78 0.039 1>
    R1 Orange:LinearColor=<rgba 0.8 0.268 0.04 1>
    R1 Pink:LinearColor=<rgba 0.505 0.0605 0.3434 1>
    R1 Purple:LinearColor=<rgba 0.1275 0.0125 0.3 1>
    R1 Saffron:LinearColor=<rgba 0.3 0.2763 0.0156 1>
    R1 SkyBlue:LinearColor=<rgba 0 1.9 2 1>
    R1 TitaniumWhite:LinearColor=<rgba 0.4 0.4 0.4 1>
    R2 Black:LinearColor=<rgba 0.0021 0.0021 0.0021 1>
    R2 BurntSienna:LinearColor=<rgba 0.17 0.0403 0.0204 1>
    R2 Cobalt:LinearColor=<rgba 0 0.011 0.22 1>
    R2 Crimson:LinearColor=<rgba 0.38 0.0057 0.0119 1>
    R2 DefaultColor:LinearColor=<rgba 1.5 0.278 0.09 1>
    R2 ForestGreen:LinearColor=<rgba 0 0.37 0.0123 1>
    R2 Grey:LinearColor=<rgba 0.038 0.038 0.038 1>
    R2 Lime:LinearColor=<rgba 0.1031 0.18 0.009 1>
    R2 Orange:LinearColor=<rgba 0.24 0.0804 0.012 1>
    R2 Pink:LinearColor=<rgba 0.15 0.018 0.102 1>
    R2 Purple:LinearColor=<rgba 0.034 0.0033 0.08 1>
    R2 Saffron:LinearColor=<rgba 0.08 0.0737 0.0042 1>
    R2 SkyBlue:LinearColor=<rgba 0 0.57 0.6 1>
    R2 TitaniumWhite:LinearColor=<rgba 0.11 0.11 0.11 1>
    Saffron=<rgba 0.9 0.8289 0.0469 1>
    SkyBlue=<rgba 0 5.7 6 1>
    Use R1 Painted Overrides=true
    Use R2 Painted Overrides=true
   )
   &VC_PaintedVariantsOverride(
    Black=<rgba 0.02 0.02 0.02 1>
    BurntSienna=<rgba 1.5 0.4875 0.15 1>
    Cobalt=<rgba 0.2 0.5634 2 1>
    Crimson=<rgba 1 0.1529 0.1356 1>
    DefaultColor=<rgba 5 3.8337 0.5 1>
    ForestGreen=<rgba 0.3 1.5 0.4274 1>
    Grey=<rgba 0.2 0.2 0.2 1>
    Lime=<rgba 0.6842 1.15 0.115 1>
    Orange=<rgba 3 0.942 0.06 1>
    Pink=<rgba 1.5 0.15 1.5 1>
    Purple=<rgba 0.17 0.0167 0.4 1>
    R1 Black:LinearColor=<rgba 0.009 0.009 0.009 1>
    R1 BurntSienna:LinearColor=<rgba 0.5 0.17 0.06 1>
    R1 Cobalt:LinearColor=<rgba 0 0.04 0.8 1>
    R1 Crimson:LinearColor=<rgba 1.2 0.036 0.0554 1>
    R1 DefaultColor=<rgba 1 0.7 0.6 1>
    R1 ForestGreen:LinearColor=<rgba 0.0817 1 0.05 1>
    R1 Grey:LinearColor=<rgba 0.15 0.15 0.15 1>
    R1 Lime:LinearColor=<rgba 0.4465 0.78 0.039 1>
    R1 Orange:LinearColor=<rgba 0.8 0.268 0.04 1>
    R1 Pink:LinearColor=<rgba 0.505 0.0605 0.3434 1>
    R1 Purple:LinearColor=<rgba 0.1275 0.0125 0.3 1>
    R1 Saffron:LinearColor=<rgba 0.3 0.2763 0.0156 1>
    R1 SkyBlue:LinearColor=<rgba 0 1.9 2 1>
    R1 TitaniumWhite:LinearColor=<rgba 0.4 0.4 0.4 1>
    R2 Black:LinearColor=<rgba 0 0 0 1>
    R2 BurntSienna:LinearColor=<rgba 0.3 0.04 0 1>
    R2 Cobalt:LinearColor=<rgba 0 0.02 0.4 1>
    R2 Crimson:LinearColor=<rgba 0.38 0.0057 0.0119 1>
    R2 DefaultColor:LinearColor=<rgba 1.5 0.278 0.09 1>
    R2 ForestGreen:LinearColor=<rgba 0 0.37 0.0123 1>
    R2 Grey:LinearColor=<rgba 0.038 0.038 0.038 1>
    R2 Lime:LinearColor=<rgba 0.1783 0.4 0.02 1>
    R2 Orange:LinearColor=<rgba 0.3 0.075 0 1>
    R2 Pink:LinearColor=<rgba 0.25 0.0299 0.17 1>
    R2 Purple:LinearColor=<rgba 0.0213 0.0021 0.05 1>
    R2 Saffron:LinearColor=<rgba 0.15 0.1382 0.0078 1>
    R2 SkyBlue:LinearColor=<rgba 0 0.57 0.6 1>
    R2 TitaniumWhite:LinearColor=<rgba 0.35 0.35 0.35 1>
    Saffron=<rgba 1 0.9167 0 1>
    SkyBlue=<rgba 0 5.7 6 1>
    TitaniumWhite=<rgba 1 1 1 1>
    Use R1 Painted Overrides=true
    Use R2 Painted Overrides=true
   ) {enabled=false, module=VC_PaintedVariantsOverride002}
   &VC_InitializeRibbon_v1(
    Color=$Particles.Color
    Custom Offset=true
    Enable Skeletal Position=true
    Lifetime=1.5
    Local Vector Offset=&AddVectorToPosition(Position=$User.DTCarOffset, Vector=<-45 0 -1>)
    Offset Amount=1
    Ribbon Width=75
   )
   &AddVelocity(Velocity=<-1000 0 0>) {module=AddVelocity001}
  ]
  particleUpdateScript [
   &ParticleState()
   &InheritSourceMovement(
    Applied Position Scale=&Multiply_VectorByFloat(Vector=<1 1 1>, Float=1)
    ApplyForce=false
    ApplyPosition=true
    ApplyPositionFalloff=false
   )
   &SolveForcesAndVelocity(Clamp Velocity=false, Limit Acceleration=false)
   &VC_UpdateRibbon_v1(Enable Skeletal Position=true)
   &ScaleRibbonWidth(Ribbon Width Scale=&FloatFromCurve(FloatCurve=DataInterface() {propertyValues {Curve {
        defaultValue=1
        keys [
         {arriveTangent=9.0725, interpMode=RCIM_Cubic, leaveTangent=9.0725, tangentMode=RCTM_User, time=0, value=0}
         {arriveTangent=-0.2271, interpMode=RCIM_Cubic, leaveTangent=-0.2271, tangentMode=RCTM_User, time=0.1408, value=0.9847}
         {arriveTangent=-1.3954, interpMode=RCIM_Cubic, leaveTangent=-1.3954, tangentMode=RCTM_User, time=1, value=0}
        ]
       }}}))
   /SetVariables(
    Particles.OutColor:LinearColor=&MultiplyLinearColor(
     Linear Color=$Particles.R2 Color
     Scale Factor=&MakeVector4FromVectorAndFloat(
      "Vector (RGB)"=&MakeCustomVectorFromBool(
       Bool=$User.bIsPainted
       True Vector=&VectorFromFloat(Value=&FloatFromCurve(
         FloatCurve=DataInterface() {propertyValues {Curve {
            defaultValue=0
            keys [{interpMode=RCIM_Cubic, tangentMode=RCTM_User, time=0.05, value=1}, {interpMode=RCIM_Cubic, tangentMode=RCTM_User, time=0.5, value=0.35}]
           }}}
         Scale Curve=2
        ))
       False Vector=&VectorFromFloat(Value=&FloatFromCurve(
         FloatCurve=DataInterface() {propertyValues {Curve {
            defaultValue=0
            keys [
             {arriveTangent=-2.8285, interpMode=RCIM_Cubic, time=0.05, value=1.8}
             {arriveTangent=-1.4947, interpMode=RCIM_Cubic, leaveTangent=-0.3961, tangentMode=RCTM_Break, time=0.5, value=0.35}
             <at 1 0.33 RCIM_Cubic>
            ]
           }}}
         Scale Curve=2
        ))
      )
      "Float (Alpha)"=1
     )
    )
    Particles.SmallerWidth:NiagaraFloat=&FloatFromCurve(
     FloatCurve=DataInterface() {propertyValues {Curve {
        defaultValue=1
        keys [
         {interpMode=RCIM_Cubic, tangentMode=RCTM_User, time=0.48, value=1}
         {arriveTangent=-3.1379, interpMode=RCIM_Cubic, leaveTangent=-3.1379, tangentMode=RCTM_User, time=0.85, value=0}
        ]
       }}}
     Scale Curve=20
    )
   ) {isSetParameters=true}
   /SetVariables(Particles.PositionL:NiagaraPosition=&AddVectorToPosition(
     Position=$Particles.Position
     Vector=&TransformVector(
      Vector=&SelectVectorFromArray(
       Direct Array Index=&SpawnGroup()
       Vector Selection Array=DataInterface() {propertyValues {FloatData [<0 -2 -2>, <0 2 -2>]}}
      )
      Destination Space=::Simulation
     )
    )) {isSetParameters=true}
  ]
  rendererClasses [&NiagaraRibbonRendererProperties]
  renderers [
   {rendererClass=&NiagaraRibbonRendererProperties}
   {rendererClass=&NiagaraRibbonRendererProperties, rendererIndex=1}
  ]
 }
 NE_VC_Boost_Sprite_Parent002 {
  emitterUpdateScript [
   &EmitterState(Inactive Response=::NewEnumerator1, Life Cycle Mode=::NewEnumerator1, UseLoopDelay=false)
   &SpawnRate(SpawnRate=30, Use Spawn Probability=false) {module=SpawnRate001}
   &VC_FilteredSocketCount()
   &SpawnRate(SpawnGroup=1, SpawnRate=15, Use Spawn Probability=false) {enabled=false}
  ]
  particleSpawnScript [
   /SetVariables(Particles.VisibilityTag=$Emitter.SpawnGroup) {isSetParameters=true}
   &VC_PaintedVariantsOverride(
    Black=<rgba 0.007 0.007 0.007 1>
    BurntSienna=<rgba 0.55 0.1953 0.077 1>
    Cobalt=<rgba 0 0.055 1.1 1>
    Crimson=<rgba 1.6 0.024 0.0503 1>
    DefaultColor=<rgba 20 6.5 2 1>
    ForestGreen=<rgba 0 1.7 0.0567 1>
    Grey=<rgba 0.18 0.18 0.18 1>
    Lime=<rgba 0.4007 0.7 0.035 1>
    Orange=<rgba 1 0.335 0.05 1>
    Pink=<rgba 0.6 0.075 0.6 1>
    Purple=<rgba 0.1275 0.0125 0.3 1>
    Saffron=<rgba 0.4 0.3684 0.0208 1>
    SkyBlue=<rgba 0 2.565 2.7 1>
    TitaniumWhite=<rgba 0.55 0.55 0.55 1>
   )
   &VC_PaintedVariantsOverride(
    Black=<rgba 0.007 0.007 0.007 1>
    BurntSienna=<rgba 1.2 0.3768 0.024 1>
    Cobalt=<rgba 0.17 0.2115 1 1>
    Crimson=<rgba 1.6 0.024 0.0503 1>
    DefaultColor=<rgba 20 6.5 2 1>
    ForestGreen=<rgba 0.2 1 0.2267 1>
    Grey=<rgba 0.08 0.08 0.08 1>
    Lime=<rgba 1.1 2 0 1>
    Orange=<rgba 2.3 0.3833 0 1>
    Pink=<rgba 2 0.3 2 1>
    Purple=<rgba 0.688 0.08 1.6 1>
    Saffron=<rgba 2 1.842 0.1042 1>
    SkyBlue=<rgba 0 2.85 3 1>
    TitaniumWhite=<rgba 2 2 2 1>
   ) {enabled=false, module=VC_PaintedVariantsOverride001}
   &InitializeParticle(
    Color=$Particles.Color
    Lifetime=0.1
    Mesh Scale Max=&MergeHelper_VectorSwitch(0=<2.5 4.5 2.5>, 1=<2.5 2.75 2.5>, 2=?, 3=?, Index=$Emitter.SpawnGroup, Num=::NewEnumerator0)
    Mesh Scale Min=&MergeHelper_VectorSwitch(0=<2.25 4 2.25>, 1=<2 2.25 2>, 2=?, 3=?, Index=$Emitter.SpawnGroup, Num=::NewEnumerator0)
    Mesh Scale Mode=::NewEnumerator2
    Sprite Rotation Mode=::NewEnumerator2
    Sprite Size Mode=::NewEnumerator3
    UsePositionOffset=false
    Write Material Random=true
   )
   &PartitionParticles(Number Of Partitions=$Emitter.PortCount)
   &SkeletalMeshLocation(
    Bone / Socket Index=$Particles.PartitionParticles.Partition
    Bone / Socket Sampling Mode=::NewEnumerator5
    MaskSpawnGroup=false
    Mesh Sampling Type=::NewEnumerator6
    Sampled Position Offset=&AddVectorToPosition(Position=$User.DTCarOffset, Vector=<-60 0 0>)
    Skeletal Mesh=$System.SkeletalMesh
   )
   &AddVelocity(Velocity=<0 0 0>) {enabled=false}
   &InitialMeshOrientation(
    Mesh Axis To Orient=<0 1 0>
    Mesh Orientation Mode=::NewEnumerator2
    Orientation Vector=<0 0 0>
    Rotation=&RandomRangeVector(Minimum=<0 -1 0>, Maximum=<0 1 0>)
    Use Rotation Vector=true
   )
  ]
  particleUpdateScript [
   &ParticleState()
   &SolveForcesAndVelocity(Clamp Velocity=false, Limit Acceleration=false)
  ]
  rendererClasses [&NiagaraMeshRendererProperties, &NiagaraSpriteRendererProperties]
  renderers [
   {rendererClass=&NiagaraMeshRendererProperties, rendererIndex=0}
   {rendererClass=&NiagaraSpriteRendererProperties, rendererIndex=1}
  ]
 }
 Rings001 {
  emitterUpdateScript [
   &EmitterState()
   &VC_FilteredSocketCount()
   &SpawnRate(SpawnRate=4, Use Spawn Probability=false)
   &SpawnRate(SpawnGroup=2, SpawnRate=4, Use Spawn Probability=false) {module=SpawnRate002}
   &SpawnRate(SpawnGroup=1, SpawnRate=40, Use Spawn Probability=false) {module=SpawnRate001}
  ]
  particleSpawnScript [
   &VC_PaintedVariantsOverride(
    Black=<rgba 0.005 0.005 0.005 1>
    BurntSienna=<rgba 0.56 0.1865 0.0633 1>
    Cobalt=<rgba 0 0.04 0.8 1>
    Crimson=<rgba 1.5 0.015 0.0397 1>
    DefaultColor=<rgba 0 0.625 2.5 1>
    ForestGreen=<rgba 0 1.5 0.05 1>
    Grey=<rgba 0.15 0.15 0.15 1>
    Lime=<rgba 0.479 0.85 0.0255 1>
    Orange=<rgba 0.86 0.328 0.043 1>
    Pink=<rgba 0.61 0.0731 0.4148 1>
    Purple=<rgba 0.12 0 0.3 1>
    R1 Black:LinearColor=<rgba 0.005 0.005 0.005 5>
    R1 BurntSienna:LinearColor=<rgba 0.56 0.1865 0.0633 1>
    R1 Cobalt:LinearColor=<rgba 0 0.04 0.8 1>
    R1 Crimson:LinearColor=<rgba 1.5 0 0.025 1>
    R1 DefaultColor=&RandomRangeLinearColor(Minimum=<rgba 2 0.2 0.1 1>, Maximum=<rgba 2 1 0.4 1>)
    R1 ForestGreen:LinearColor=<rgba 0 1.5 0.05 1>
    R1 Grey:LinearColor=<rgba 0.13 0.13 0.13 1>
    R1 Lime:LinearColor=<rgba 0.42 0.7 0 2>
    R1 Orange:LinearColor=<rgba 0.8 0.279 0 1>
    R1 Pink:LinearColor=<rgba 1.3 0 0.8273 1>
    R1 Purple:LinearColor=<rgba 0.084 0 0.21 1>
    R1 Saffron:LinearColor=<rgba 0.3 0.275 0 1>
    R1 SkyBlue:LinearColor=<rgba 0 1.9 2 1>
    R1 TitaniumWhite:LinearColor=<rgba 0.5 0.5 0.5 1>
    R2 DefaultColor:LinearColor=<rgba 0.4 1.2 2 1>
    Saffron=<rgba 0.36 0.3316 0.0188 1>
    SkyBlue=<rgba 0 2.375 2.5 1>
    TitaniumWhite=<rgba 0.5 0.5 0.5 1>
    Use R1 Painted Overrides=true
    Use R2 Painted Overrides=true
   )
   &VC_PaintedVariantsOverride(
    Black=<rgba 0.02 0.02 0.02 1>
    BurntSienna=<rgba 1.5 0.4875 0.15 1>
    Cobalt=<rgba 0.2 0.5634 2 1>
    Crimson=<rgba 1 0.1529 0.1356 1>
    DefaultColor=<rgba 5 3.8337 0.5 1>
    ForestGreen=<rgba 0.3 1.5 0.4274 1>
    Grey=<rgba 0.2 0.2 0.2 1>
    Lime=<rgba 0.6842 1.15 0.115 1>
    Orange=<rgba 3 0.942 0.06 1>
    Pink=<rgba 1.5 0.15 1.5 1>
    Purple=<rgba 0.17 0.0167 0.4 1>
    R1 Black:LinearColor=<rgba 0.009 0.009 0.009 1>
    R1 BurntSienna:LinearColor=<rgba 0.5 0.17 0.06 1>
    R1 Cobalt:LinearColor=<rgba 0 0.04 0.8 1>
    R1 Crimson:LinearColor=<rgba 1.2 0.036 0.0554 1>
    R1 DefaultColor=<rgba 1 0.7 0.6 1>
    R1 ForestGreen:LinearColor=<rgba 0.0817 1 0.05 1>
    R1 Grey:LinearColor=<rgba 0.15 0.15 0.15 1>
    R1 Lime:LinearColor=<rgba 0.4465 0.78 0.039 1>
    R1 Orange:LinearColor=<rgba 0.8 0.268 0.04 1>
    R1 Pink:LinearColor=<rgba 0.505 0.0605 0.3434 1>
    R1 Purple:LinearColor=<rgba 0.1275 0.0125 0.3 1>
    R1 Saffron:LinearColor=<rgba 0.3 0.2763 0.0156 1>
    R1 SkyBlue:LinearColor=<rgba 0 1.9 2 1>
    R1 TitaniumWhite:LinearColor=<rgba 0.4 0.4 0.4 1>
    R2 Black:LinearColor=<rgba 0 0 0 1>
    R2 BurntSienna:LinearColor=<rgba 0.3 0.04 0 1>
    R2 Cobalt:LinearColor=<rgba 0 0.02 0.4 1>
    R2 Crimson:LinearColor=<rgba 0.38 0.0057 0.0119 1>
    R2 DefaultColor:LinearColor=<rgba 1.5 0.278 0.09 1>
    R2 ForestGreen:LinearColor=<rgba 0 0.37 0.0123 1>
    R2 Grey:LinearColor=<rgba 0.038 0.038 0.038 1>
    R2 Lime:LinearColor=<rgba 0.1783 0.4 0.02 1>
    R2 Orange:LinearColor=<rgba 0.3 0.075 0 1>
    R2 Pink:LinearColor=<rgba 0.25 0.0299 0.17 1>
    R2 Purple:LinearColor=<rgba 0.0213 0.0021 0.05 1>
    R2 Saffron:LinearColor=<rgba 0.15 0.1382 0.0078 1>
    R2 SkyBlue:LinearColor=<rgba 0 0.57 0.6 1>
    R2 TitaniumWhite:LinearColor=<rgba 0.35 0.35 0.35 1>
    Saffron=<rgba 1 0.9167 0 1>
    SkyBlue=<rgba 0 5.7 6 1>
    TitaniumWhite=<rgba 1 1 1 1>
    Use R1 Painted Overrides=true
    Use R2 Painted Overrides=true
   ) {enabled=false, module=VC_PaintedVariantsOverride001}
   &InitializeParticle(
    Color=$Particles.Color
    Lifetime=&MergeHelper_FloatSwitch(0=1, 1=0.8, 2=1, 3=?, Index=$Emitter.SpawnGroup, Num=::NewEnumerator1)
    Mesh Scale Mode=::NewEnumerator3
    Mesh Uniform Scale=&RandomRangeFloat(Minimum=10.5, Maximum=11.5)
    Sprite Size Mode=::NewEnumerator1
    Uniform Sprite Size Max=14
    Uniform Sprite Size Min=10
    UsePositionOffset=false
    Write Material Random=true
   )
   &PartitionParticles(Number Of Partitions=$Emitter.PortCount)
   &SkeletalMeshLocation(
    Bone / Socket Index=$Particles.PartitionParticles.Partition
    Bone / Socket Sampling Mode=::NewEnumerator5
    MaskSpawnGroup=false
    Mesh Sampling Type=::NewEnumerator6
    Sampled Position Offset=&AddVectorToPosition(
     Position=$User.DTCarOffset
     Vector=&MergeHelper_VectorSwitch(0=<-60 0 0>, 1=<-80 0 0>, 2=<-60 0 0>, 3=?, Index=$Emitter.SpawnGroup, Num=::NewEnumerator1)
    )
    Skeletal Mesh=$System.SkeletalMesh
   )
   &AddVelocity(Velocity=&MergeHelper_VectorSwitch(0=<-1000 0 0>, 1=<-1000 0 0>, 2=<-1000 0 0>, 3=?, Index=$Emitter.SpawnGroup, Num=::NewEnumerator1))
   &ShapeLocation(
    Box Size=<90 90 90>
    MaskSpawnGroup=false
    Non Uniform Scale=&MergeHelper_VectorSwitch(0=<0 0 0>, 1=<0.5 0.5 0.5>, 2=<0 0 0>, 3=?, Index=$Emitter.SpawnGroup, Num=::NewEnumerator1)
    Override Seed=false
    Shape Primitive=::NewEnumerator2
    Surface Only=false
    UseOwnerScale=true
   )
   &InitialMeshOrientation(
    Mesh Axis To Orient=<0 0 1>
    Mesh Orientation Mode=::NewEnumerator2
    Orientation Coordinate Space=::Local
    Orientation Vector=<0 0 1>
    Use Rotation Vector=false
   )
   /SetVariables(Particles.VisibilityTag=$Emitter.SpawnGroup) {isSetParameters=true}
  ]
  particleUpdateScript [
   &ParticleState()
   &ScaleSpriteSize(
    Uniform Curve Scale=0.8
    Uniform Curve Sprite Scale=DataInterface() {propertyValues {Curve {keys [{tangentMode=RCTM_User, time=0, value=1}, <at 0.4 1>, {tangentMode=RCTM_User, time=1, value=0}]}}}
   )
   &ScaleMeshSize(Scale Factor=&Multiply_VectorByFloat(
     Vector=<0.8 0.8 0.8>
     Float=&FloatFromCurve(FloatCurve=DataInterface() {propertyValues {Curve {keys [
          <at 0 0.6 RCIM_Cubic>
          {arriveTangent=1.9655, interpMode=RCIM_Cubic, leaveTangent=1.9655, time=0.3854, value=1.9}
          <at 0.8141 2.2 RCIM_Cubic>
         ]}}})
    ))
   &AccelerationForce(Acceleration=&MergeHelper_VectorSwitch(
     0=<0 0 0>
     1=&Multiply_VectorByFloat(
      Vector=&RandomRangeVector(Minimum=<0 -200 -200>, Maximum=<0 200 200>)
      Float=&FloatFromCurve(
       FloatCurve=DataInterface() {propertyValues {Curve {keys [
           {interpMode=RCIM_Cubic, tangentMode=RCTM_User, time=0, value=0}
           {interpMode=RCIM_Cubic, tangentMode=RCTM_User, time=0.33, value=2}
           {arriveTangent=-2.9851, interpMode=RCIM_Cubic, leaveTangent=-2.9851, time=0.66, value=-2}
           {tangentMode=RCTM_User, time=1, value=0}
          ]}}}
       Scale Curve=2
      )
     )
     2=?
     3=?
     Index=&SpawnGroup()
     Num=::NewEnumerator0
    ))
   &InheritSourceMovement(
    Applied Position Scale=&Multiply_VectorByFloat(Vector=<1 1 1>, Float=1)
    ApplyForce=false
    ApplyPosition=true
    ApplyPositionFalloff=false
   )
   &SolveForcesAndVelocity(Clamp Velocity=false, Limit Acceleration=false)
   &AlignSpriteToMeshOrientation(
    Mesh Orientation Relative Sprite Facing Vector:Vector3f=&SelectVectorFromArray(
     Direct Array Index=&SpawnGroup()
     Vector Selection Array=DataInterface() {propertyValues {FloatData [<1 0 0>, <-1 0 0>]}}
    )
    SetAlignment=true
    SetFacing=true
   )
   &ScaleColor(
    Scale Alpha=&FloatFromCurve(FloatCurve=DataInterface() {propertyValues {Curve {keys [<at 0.5 1>,<at 1 0>]}}})
    ScaleA=true
    ScaleRGB=false
   )
   &DynamicMaterialParameters(
    Index 1 Param 1=&FloatFromCurve(FloatCurve=DataInterface() {propertyValues {Curve {keys [
         {arriveTangent=-13.683, interpMode=RCIM_Cubic, leaveTangent=-13.683, tangentMode=RCTM_User, time=0, value=0}
         {arriveTangent=-4.0398, interpMode=RCIM_Cubic, leaveTangent=1.6744, tangentMode=RCTM_Break, time=0.2, value=-1}
         {arriveTangent=5.0761, interpMode=RCIM_Cubic, leaveTangent=5.0761, tangentMode=RCTM_User, time=0.8, value=1.5}
        ]}}})
    Index 1 Param 2=1
    Module.Index0Float0Write=true
    Module.Index0Float1Write=true
    Module.Index0Float2Write=true
    Module.Index0Float3Write=true
    Module.Index1Float0Write=true
    Module.Index1Float1Write=true
    Module.Index1Float2Write=true
    Module.Index1Float3Write=true
    Write Parameter Index 1=::NewEnumerator1
   ) {module=DynamicMaterialParameters001}
   /SetVariables(
    Particles.GlowColor=&MultiplyLinearColor(
     Linear Color=$Particles.R2 Color
     Scale Factor=&MakeVector4FromVectorAndFloat(
      "Vector (RGB)"=&VectorFromFloat(Value=15)
      "Float (Alpha)"=&FloatFromCurve(
       FloatCurve=DataInterface() {propertyValues {Curve {
          defaultValue=1
          keys [{interpMode=RCIM_Cubic, tangentMode=RCTM_User, time=0.3881, value=1}, {interpMode=RCIM_Cubic, tangentMode=RCTM_User, time=0.55, value=0}]
         }}}
       Scale Curve=&MakeCustomFloatFromBool(Bool=$User.bIsPainted, True Float=0, False Float=0.01)
      )
     )
    )
    Particles.GlowSize=&Vector2DFromFloat(Value=&FloatFromCurve(
      FloatCurve=DataInterface() {propertyValues {Curve {
         defaultValue=1
         keys [
          {interpMode=RCIM_Cubic, tangentMode=RCTM_User, time=0, value=0.4}
          {arriveTangent=1, interpMode=RCIM_Cubic, leaveTangent=1, time=0.4, value=1}
          <at 0.8 1.2 RCIM_Cubic>
         ]
        }}}
      Scale Curve=210
     ))
   ) {isSetParameters=true}
  ]
  rendererClasses [&NiagaraMeshRendererProperties, &NiagaraSpriteRendererProperties]
  renderers [
   {rendererClass=&NiagaraMeshRendererProperties, rendererIndex=0}
   {rendererClass=&NiagaraSpriteRendererProperties, rendererIndex=1}
   {rendererClass=&NiagaraSpriteRendererProperties, rendererIndex=2}
   {rendererClass=&NiagaraSpriteRendererProperties, rendererIndex=3}
  ]
 }
 ribbon {
  emitterUpdateScript [
   &EmitterState(Inactive Response=::NewEnumerator1, Life Cycle Mode=::NewEnumerator1, Loop Duration=1, UseLoopDelay=false)
   &VC_FilteredSocketCount()
   &VC_RateQualityHandler()
   &VC_SpawnRibbon_v1(Number of Ribbons=$Emitter.PortCount, Spawn Time=-1, SpawnRate=60)
  ]
  particleSpawnScript [
   &VC_PaintedVariantsOverride(
    Black=<rgba 0 0 0 1>
    BurntSienna=<rgba 0.6 0.12 0 1>
    Cobalt=<rgba 0 0 0.4 1>
    Crimson=<rgba 0.4 0 0 1>
    DefaultColor=<rgba 0 1.2 2 1>
    ForestGreen=<rgba 0 0.8 0 1>
    Grey=<rgba 0.05 0.05 0.05 10>
    Lime=<rgba 0.5852 1 0 1>
    Orange=<rgba 1 0.3618 0.02 1>
    Pink=<rgba 0.8 0.12 0.5527 1>
    Purple=<rgba 0.48 0 1.2 1>
    Saffron=<rgba 1 0.9167 0 1>
    SkyBlue=<rgba 0 0.95 1 1>
    TitaniumWhite=<rgba 1 1 1 1>
   )
   &VC_InitializeRibbon_v1(
    Color=$Particles.Color
    Custom Offset=true
    Enable Skeletal Position=true
    Lifetime=0.8
    Local Vector Offset=&AddVectorToPosition(Position=<-50 0 0>, Vector=$User.DTCarOffset001)
    Offset Amount=1
    Write Material Random=false
   )
   &AddVelocity(Velocity=<1750 0 0>, Velocity Speed Scale=-1) {module=AddVelocity001}
   &ShapeLocation(MaskSpawnGroup=false, Sphere Distribution=::NewEnumerator1, Sphere Radius=10, UseOwnerScale=true)
  ]
  particleUpdateScript [
   &ParticleState()
   &InheritSourceMovement(
    Applied Position Scale=&Multiply_VectorByFloat(Vector=<1 1 1>, Float=1)
    ApplyForce=false
    ApplyPosition=true
    ApplyPositionFalloff=false
   )
   &SolveForcesAndVelocity(Clamp Velocity=false, Limit Acceleration=false)
   &VC_UpdateRibbon_v1(Enable Skeletal Position=true)
   &ScaleRibbonWidth(Ribbon Width Scale=&FloatFromCurve(FloatCurve=DataInterface() {propertyValues {Curve {keys [<at 0.3061 1>,<at 1 0>]}}}))
   &DynamicMaterialParameters(Module.Index0Float0Write=true, Module.Index0Float1Write=true, Module.Index0Float2Write=true, Module.Index0Float3Write=true)
   /SetVariables(
    Particles.PosNoodle1:NiagaraPosition=&AddVectorToPosition(
     Position=$Particles.Position
     Vector=&TransformVector(
      Vector=&MakeVector(
       X=0
       Y=&Sine(
        Normalized Angle=&Add_Float(
         A=&Multiply_Float(A=$Particles.NormalizedAge, B=-1.4)
         B=&Multiply_Float(A=$Emitter.LoopedAge, B=1)
        )
        Period=1
        Scale=&FloatFromCurve(
         FloatCurve=DataInterface() {propertyValues {Curve {defaultValue=1, keys [{tangentMode=RCTM_User, time=0, value=0}, {tangentMode=RCTM_User, time=0.53, value=1.1}]}}}
         Scale Curve=17
        )
        Bias=0
       )
       Z=&Sine(
        Normalized Angle=&Add_Float(
         A=&Multiply_Float(A=$Particles.NormalizedAge, B=-1.6)
         B=&Multiply_Float(A=$Emitter.LoopedAge, B=1.5)
        )
        Period=0.75
        Scale=&FloatFromCurve(
         FloatCurve=DataInterface() {propertyValues {Curve {
            defaultValue=1
            keys [{interpMode=RCIM_Cubic, tangentMode=RCTM_User, time=0, value=1}, {interpMode=RCIM_Cubic, tangentMode=RCTM_User, time=1, value=0.34}]
           }}}
         Scale Curve=30
        )
        Bias=0
       )
      )
      Destination Space=::Simulation
      Apply Scale=true
     )
    )
    Particles.PosNoodle2:NiagaraPosition=&AddVectorToPosition(
     Position=$Particles.Position
     Vector=&TransformVector(
      Vector=&MakeVector(
       X=0
       Y=&Sine(
        Normalized Angle=&Add_Float(
         A=&Multiply_Float(A=$Particles.NormalizedAge, B=-1.4)
         B=&Multiply_Float(A=$Emitter.LoopedAge, B=1)
        )
        Period=1
        Scale=&FloatFromCurve(
         FloatCurve=DataInterface() {propertyValues {Curve {defaultValue=1, keys [{tangentMode=RCTM_User, time=0, value=0}, {tangentMode=RCTM_User, time=0.53, value=1.1}]}}}
         Scale Curve=17
        )
        Bias=0
       )
       Z=&Sine(
        Normalized Angle=&Add_Float(
         A=&Multiply_Float(A=$Particles.NormalizedAge, B=-1.6)
         B=&Multiply_Float(A=$Emitter.LoopedAge, B=1.5)
        )
        Period=0.75
        Scale=&FloatFromCurve(
         FloatCurve=DataInterface() {propertyValues {Curve {
            defaultValue=1
            keys [{interpMode=RCIM_Cubic, tangentMode=RCTM_User, time=0, value=1}, {interpMode=RCIM_Cubic, tangentMode=RCTM_User, time=1, value=0.34}]
           }}}
         Scale Curve=-30
        )
        Bias=0
       )
      )
      Destination Space=::Simulation
      Apply Scale=true
     )
    )
   ) {isSetParameters=true}
  ]
  rendererClasses [&NiagaraRibbonRendererProperties]
  renderers [
   {rendererClass=&NiagaraRibbonRendererProperties}
   {rendererClass=&NiagaraRibbonRendererProperties, rendererIndex=1}
  ]
 }
}
summary {
 systemName=NS_Photon
 emitters [
  {emitterName=Rings001, rendererClasses [&NiagaraMeshRendererProperties, &NiagaraSpriteRendererProperties]}
  {emitterName=ribbon, rendererClasses [&NiagaraRibbonRendererProperties]}
  {
   emitterName=NE_VC_Boost_Sprite_Parent002
   rendererClasses [&NiagaraMeshRendererProperties, &NiagaraSpriteRendererProperties]
  }
  {emitterName=L_Second_Ribbon001, rendererClasses [&NiagaraRibbonRendererProperties]}
 ]
 userVariables [
  User.bIsFrontendPreview {type=NiagaraBool}
  User.bIsPainted {type=NiagaraBool}
  User.ColorVariantID {type=NiagaraInt32}
  User.DTCarOffset {type=NiagaraPosition, description="", defaultValue {}}
  User.DTCarOffset001 {type=Vector3f, description="", defaultValue {struct=Vector3f, value=<0 0 0>}}
  User.DTCarOffset002 {type=Vector3f, description="", defaultValue {struct=Vector3f, value=<0 0 0>}}
  User.LutColor {type=NiagaraFloat, defaultValue=0.45, description=""}
  User.PaintedColor {type=LinearColor, description="", defaultValue {struct=LinearColor, value=<rgba 1 1 1 1>}}
  User.PaintedScalar {type=NiagaraFloat, defaultValue=1, description=""}
  User.PortCount {type=NiagaraInt32, defaultValue=1}
 ]
}
